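; Test spline bridge actor that builds its deck and optional support columns in the construction script; its event graph holds only empty events.
; It divides the spline by section length into collidable spline mesh segments, using GetStartPos and GetEndPos for positions and length-clamped tangents.
; When AddColumns is set, evenly spaced column meshes drop to RoadHeight and scale vertically by ColumnMeshHeight to reach the spline.
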(event EventBeginPlay)

(event Collision|EventActorBeginOverlap
  (params (OtherActor :type Actor%20Object%20Reference)))

(event EventTick
  (params (DeltaSeconds :type Float%20%28single-precision%29)))

(fn GetEndPos
  (params (Index :type Integer))
  (return (Spline|GetLocationatDistanceAlongSpline (Variables|Default|GetSpline) (* (+ Index 1) (Variables|Default|GetSectionLength)) "Local") (Math|Vector|ClampVectorSize (Spline|GetTangentatDistanceAlongSpline (Variables|Default|GetSpline) (* (+ Index 1) (Variables|Default|GetSectionLength)) "Local") 0.0 (Variables|Default|GetSectionLength))))

(fn GetStartPos
  (params (Index :type Integer))
  (return (Spline|GetLocationatDistanceAlongSpline (Variables|Default|GetSpline) (* Index (Variables|Default|GetSectionLength)) "Local") (Math|Vector|ClampVectorSize (Spline|GetTangentatDistanceAlongSpline (Variables|Default|GetSpline) (* Index (Variables|Default|GetSectionLength)) "Local") 0.0 (Variables|Default|GetSectionLength))))

(fn ConstructionScript
  (for _index (range (+ (Math|Float|Truncate (/ (Spline|GetSplineLength (Variables|Default|GetSpline)) (Variables|Default|GetSectionLength))) 1))
    (Variables|Default|SetSplineLoopIndex _index)
    (bind _returnvalue (AddComponent|NODE_AddSplineMeshComponent-0))
    (Components|StaticMesh|SetStaticMesh _returnvalue (Variables|Default|GetSplineMesh))
    (SplineMesh|SetForwardAxis _returnvalue "X" true)
    (bind (_position _tangent) (CallFunction|GetStartPos (Variables|Default|GetSplineLoopIndex)))
    (bind (_position_2 _tangent_2) (CallFunction|GetEndPos (Variables|Default|GetSplineLoopIndex)))
    (SplineMesh|SetStartandEnd _returnvalue _position _tangent _position_2 _tangent_2 true :!class /Script/Engine.SplineMeshComponent)
    (Collision|SetCollisionEnabled _returnvalue "QueryAndPhysics"))
  (if (Variables|Default|GetAddColumns)
    (Variables|Default|SetDistBetweenColumns (Math|Float|Truncate (Math|Float|SafeDivide (Spline|GetSplineLength (Variables|Default|GetSpline)) (Math|Conversions|ToFloat(Integer) (Variables|Default|GetNumColumns)))))
    (for _index_2 (range (Variables|Default|GetNumColumns))
      (bind _output_get (Variables|Default|SetColumnLoopIndex _index_2))
      (bind _returnvalue_2 (AddComponent|NODE_AddStaticMeshComponent-1 :RelativeTransform (Math|Transform|MakeTransform (Spline|GetLocationatDistanceAlongSpline (Variables|Default|GetSpline) (Math|Conversions|ToFloat(Integer) (* (+ _output_get 1) (Variables|Default|GetDistBetweenColumns))) "Local"))))
      (Transformation|SetWorldTransform _returnvalue_2 (Math|Transform|MakeTransform (Math|Vector|MakeVector (.x (Transformation|GetWorldLocation _returnvalue_2)) (.y (Transformation|GetWorldLocation _returnvalue_2)) (Variables|Default|GetRoadHeight)) "0, 0, 0" (Math|Vector|MakeVector 1.0 1.0 (/ (- (.z (Transformation|GetWorldLocation _returnvalue_2)) (Variables|Default|GetRoadHeight)) (Variables|Default|GetColumnMeshHeight)))) false false)
      (Components|StaticMesh|SetStaticMesh _returnvalue_2 (Variables|Default|GetColumnMesh)))))
