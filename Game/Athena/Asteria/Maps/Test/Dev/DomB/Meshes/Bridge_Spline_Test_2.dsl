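; Test bridge Blueprint that builds a spline-based structure in its construction script; its event graph is empty.
; Spline mesh segments tile the main spline at a set section length, with tangents clamped to that length.
; Optionally it adds evenly spaced columns Z-scaled between spline and road height, or else sub-splines dropping to road height with X/Y offsets and set tangents.
; Each sub-spline gets its own spline meshes.
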(event EventBeginPlay)

(event Collision|EventActorBeginOverlap
  (params (OtherActor :type Actor%20Object%20Reference)))

(event EventTick
  (params (DeltaSeconds :type Float%20%28single-precision%29)))

(fn AddMeshesToSubSplines
  (params (SubSpline :type Spline%20Component%20Object%20Reference))
  (for _index (range (+ (Math|Float|Truncate (/ (Spline|GetSplineLength SubSpline) (Variables|Default|GetSubSplineMeshSectionLength))) 1))
    (LocalVariables|SetLoopIndex _index)
    (bind _returnvalue (AddComponent|NODE_AddSplineMeshComponent-2)) ; add mesh to sub splines
    (Components|StaticMesh|SetStaticMesh _returnvalue (Variables|Default|GetSubSplineMesh))
    (SplineMesh|SetForwardAxis _returnvalue "Z" true)
    (bind (_position _tangent) (CallFunction|GetStartPos (LocalVariables|GetLoopIndex) SubSpline (Variables|Default|GetSubSplineMeshSectionLength)))
    (bind (_position_2 _tangent_2) (CallFunction|GetEndPos (LocalVariables|GetLoopIndex) SubSpline (Variables|Default|GetSubSplineMeshSectionLength)))
    (SplineMesh|SetStartandEnd _returnvalue _position _tangent _position_2 _tangent_2 true :!class /Script/Engine.SplineMeshComponent)
    (Collision|SetCollisionEnabled _returnvalue "QueryAndPhysics")))

(fn GetEndPos
  (params (Index :type Integer) (SplineRef :type Spline%20Component%20Object%20Reference) (SplineSectionLength :type Float%20%28double-precision%29))
  (return (Spline|GetLocationatDistanceAlongSpline SplineRef (* (+ Index 1) SplineSectionLength) "Local") (Math|Vector|ClampVectorSize (Spline|GetTangentatDistanceAlongSpline SplineRef (* (+ Index 1) SplineSectionLength) "Local") 0.0 (Variables|Default|GetSectionLength))))

(fn GetStartPos
  (params (Index :type Integer) (Spline%20Ref :type Spline%20Component%20Object%20Reference) (SplineSectionLength :type Float%20%28double-precision%29))
  (return (Spline|GetLocationatDistanceAlongSpline Spline%20Ref (* Index SplineSectionLength) "Local") (Math|Vector|ClampVectorSize (Spline|GetTangentatDistanceAlongSpline Spline%20Ref (* Index SplineSectionLength) "Local") 0.0 (Variables|Default|GetSectionLength))))

(fn ConstructionScript
  (for _index (range (+ (Math|Float|Truncate (/ (Spline|GetSplineLength (Variables|Default|GetSpline)) (Variables|Default|GetSectionLength))) 1))
    (Variables|Default|SetSplineLoopIndex _index)
    (bind _returnvalue (AddComponent|NODE_AddSplineMeshComponent-0))
    (Components|StaticMesh|SetStaticMesh _returnvalue (Variables|Default|GetSplineMesh))
    (SplineMesh|SetForwardAxis _returnvalue "X" true)
    (bind (_position _tangent) (CallFunction|GetStartPos (Variables|Default|GetSplineLoopIndex) (Variables|Default|GetSpline) (Variables|Default|GetSectionLength)))
    (bind (_position_2 _tangent_2) (CallFunction|GetEndPos (Variables|Default|GetSplineLoopIndex) (Variables|Default|GetSpline) (Variables|Default|GetSectionLength)))
    (SplineMesh|SetStartandEnd _returnvalue _position _tangent _position_2 _tangent_2 true :!class /Script/Engine.SplineMeshComponent)
    (Collision|SetCollisionEnabled _returnvalue "QueryAndPhysics"))
  (if (Variables|Default|GetAddColumns)
    (Variables|Default|SetDistBetweenColumns (Math|Float|Truncate (Math|Float|SafeDivide (Spline|GetSplineLength (Variables|Default|GetSpline)) (Math|Conversions|ToFloat(Integer) (Variables|Default|GetNumColumns)))))
    (for _index_2 (range (Variables|Default|GetNumColumns))
      (bind _output_get (Variables|Default|SetColumnLoopIndex _index_2))
      (bind _returnvalue_2 (AddComponent|NODE_AddStaticMeshComponent-1 :RelativeTransform (Math|Transform|MakeTransform (Spline|GetLocationatDistanceAlongSpline (Variables|Default|GetSpline) (Math|Conversions|ToFloat(Integer) (* (+ _output_get 1) (Variables|Default|GetDistBetweenColumns))) "Local"))))
      (Transformation|SetWorldTransform _returnvalue_2 (Math|Transform|MakeTransform (Math|Vector|MakeVector (.x (Transformation|GetWorldLocation _returnvalue_2)) (.y (Transformation|GetWorldLocation _returnvalue_2)) (Variables|Default|GetRoadHeight)) "0, 0, 0" (Math|Vector|MakeVector 1.0 1.0 (/ (- (.z (Transformation|GetWorldLocation _returnvalue_2)) (Variables|Default|GetRoadHeight)) (Variables|Default|GetColumnMeshHeight)))) false false)
      (Components|StaticMesh|SetStaticMesh _returnvalue_2 (Variables|Default|GetColumnMesh)))
    (elif (Variables|Default|GetAddSplines)
      (Variables|Default|SetDistBetweenColumns (Math|Float|Truncate (Math|Float|SafeDivide (Spline|GetSplineLength (Variables|Default|GetSpline)) (Math|Conversions|ToFloat(Integer) (Variables|Default|GetNumColumns)))))
      (for _index_3 (range (Variables|Default|GetNumColumns))
        (bind _output_get_2 (Variables|Default|SetColumnLoopIndex _index_3))
        (bind _returnvalue_3 (AddComponent|NODE_AddSplineComponent-0))
        (if (Variables|Default|GetUseLocalOffset)
          (Spline|SetSplinePoints _returnvalue_3 (Utilities|Array|MakeArray (Spline|GetLocationatDistanceAlongSpline (Variables|Default|GetSpline) (Math|Conversions|ToFloat(Integer) (* (+ _output_get_2 1) (Variables|Default|GetDistBetweenColumns))) "World") (Math|Vector|MakeVector (+ (.x (Spline|GetLocationatDistanceAlongSpline (Variables|Default|GetSpline) (Math|Conversions|ToFloat(Integer) (* (+ _output_get_2 1) (Variables|Default|GetDistBetweenColumns))) "World")) (Variables|Default|GetSubSplineOffsetX)) (+ (.y (Spline|GetLocationatDistanceAlongSpline (Variables|Default|GetSpline) (Math|Conversions|ToFloat(Integer) (* (+ _output_get_2 1) (Variables|Default|GetDistBetweenColumns))) "World")) (Variables|Default|GetSubSplineOffsetY)) (Variables|Default|GetRoadHeight))) "World" true)
          (else
            (Spline|SetSplinePoints _returnvalue_3 (Utilities|Array|MakeArray (Spline|GetLocationatDistanceAlongSpline (Variables|Default|GetSpline) (Math|Conversions|ToFloat(Integer) (* (+ _output_get_2 1) (Variables|Default|GetDistBetweenColumns))) "World") (Math|Vector|MakeVector (+ (.x (Spline|GetLocationatDistanceAlongSpline (Variables|Default|GetSpline) (Math|Conversions|ToFloat(Integer) (* (+ _output_get_2 1) (Variables|Default|GetDistBetweenColumns))) "World")) (Variables|Default|GetSubSplineOffsetX)) (+ (Variables|Default|GetSubSplineOffsetY) (.y (Spline|GetLocationatSplinePoint (Variables|Default|GetSpline) 0 "World"))) (Variables|Default|GetRoadHeight))) "World" true)))
        (Spline|SetTangentatSplinePoint _returnvalue_3 1 (Variables|Default|GetSubSplineBottomVector) "Local" true)
        (Spline|SetTangentatSplinePoint _returnvalue_3 0 (Variables|Default|GetSubSplineTopVector) "Local" true)
        (CallFunction|AddMeshestoSubSplines _returnvalue_3)))))
