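# Translucent, default-lit material built on a crescent texture, kept in a developer test map folder under Athena Asteria.
# The graph samples the crescent texture and multiplies its alpha channel by a constant to halve it.
# A second sample node also reads a crescent texture, and no links from either branch to the material outputs are shown.
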Material /Game/Athena/Asteria/Maps/Test/Dev/DomB/m_crescent
graph {
 properties {blendMode=BLEND_Translucent, pixelDepthOffsetMode=PDOM_Legacy, usedShadingModels=Default Lit}
 nodes [
  Multiply(A=TextureSample() {texture=&/crescent}.A) {constB=0.5}
  TextureSample() {texture=&crescent}
 ]
}
