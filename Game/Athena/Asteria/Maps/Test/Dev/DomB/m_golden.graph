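# Translucent Default Lit material in a developer test map folder, named for gold or possibly a golden-ratio guide.
# Its only node is a texture sample of Ratio-1024, with no connection to any material output shown.
# No parameters, math or other inputs are present, so it appears to be an unfinished test material.
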Material /Game/Athena/Asteria/Maps/Test/Dev/DomB/m_golden
graph {
 properties {blendMode=BLEND_Translucent, pixelDepthOffsetMode=PDOM_Legacy, usedShadingModels=Default Lit}
 nodes [TextureSample() {texture=&/Ratio-1024}]
}
